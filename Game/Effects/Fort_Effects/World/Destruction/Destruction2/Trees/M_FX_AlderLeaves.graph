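# Two-sided subsurface material for Niagara mesh particle alder leaves in tree destruction effects, with burning looks.
# Diffuse blends a particle-tinted leaf texture with a colorized one, then darkens toward a desaturated burn-gradient result via the Burnt dynamic parameter.
# An OSSR texture drives specular, roughness and opacity masking, with an optional extra mask switch.
# Glow scales emissive from a fiery curve atlas, fed by texture reads plus Fresnel, with HDR derivation on the default shading path.
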Material /Game/Effects/Fort_Effects/World/Destruction/Destruction2/Trees/M_FX_AlderLeaves
graph {
 properties {
  bUsedWithNiagaraMeshParticles=true
  pixelDepthOffsetMode=PDOM_Legacy
  shadingModel=MSM_Subsurface
  twoSided=true
  usedShadingModels=Subsurface
 }
 nodes [
  n1 = NamedRerouteUsage()
  n2 = Constant() {r=0.5}
  LinearInterpolate(
   A=NamedRerouteUsage()
   B=Multiply(A=NamedRerouteUsage(), B=Desaturation(None=@n1, Fraction=@n2))
   Alpha=DynamicParameter() {defaultValue=<rgba 0 0 1 1>, paramNames [Glow,Burnt,Param3,Param4]}.Burnt
  )
  n3 = TextureSampleParameter2D() {parameterName=OSSR Texture, samplerType=SAMPLERTYPE_Masks, texture=/Game/Environments/Asteria/Foliage/Trees/Oak/Textures/T_AlderLeafs2_OSSR}
  LinearInterpolate(B=ScalarParameter() {defaultValue=1, parameterName=Max Specular}, Alpha=@n3.R)
  LinearInterpolate(B=ScalarParameter() {defaultValue=1, parameterName=Max Roughness}, Alpha=@n3.B)
  n4 = TextureSampleParameter2D() {parameterName=Mask Texture, samplerType=SAMPLERTYPE_Masks, texture=/Game/Global/Textures/Default/Blanks/AllRed}
  Multiply(
   A=Multiply(A=@n3.G) {constB=0.9, desc=Estimated value from Ateria Leaves Base Material}
   B=StaticSwitchParameter(True=ChannelMaskParameter(None=@n4.RGBA) {parameterName=Mask Texture Channel}.R, False=Constant() {r=1}) {parameterName=Has Mask Texture}
  )
  Multiply(A=NamedRerouteUsage(), B=ScalarParameter() {defaultValue=0.2, parameterName=SubSufeace Multiplier}) {constB=0.2, desc=Estimated value from Ateria Leaves Base Material}
  n5 = TextureSampleParameter2D() {parameterName=Diffuse Leaf Texture, texture=/Game/Environments/Asteria/Foliage/Trees/Oak/Textures/T_AlderLeafs2_D2}
  n6 = ParticleColor()
  NamedRerouteDeclaration(None=LinearInterpolate(
    A=Multiply(A=@n5.RGB, B=@n6.RGB)
    B=TextureSampleParameter2D() {parameterName=Diffuse Leaf Texture_Color, texture=/Game/Environments/Asteria/Foliage/Trees/Oak/Textures/T_AlderLeafs2_D}.RGB
    Alpha=ScalarParameter() {parameterName=Use Colorized Texture}
   )) {name=Diffuse, nodeColor=<rgba 1 0.4688 0 1>}
  n7 = TextureSample() {
   samplerType=SAMPLERTYPE_LinearGrayscale
   texture=/Game/Characters/Player/Male/Medium/Bodies/M_Med_Soldier_04/Skins/Gingerbread_Burnt/Textures/BurnGradient_output
  }
  NamedRerouteDeclaration(None=LinearInterpolate(Alpha=@n7.G) {constA=0.05, constB=0.9}) {name=Burnt Sample, nodeColor=<rgba 0.2422 1 0 1>}
  n8 = TextureSampleParameter2D() {parameterName=Normal Texture, samplerType=SAMPLERTYPE_Normal, texture=/Game/Environments/Asteria/Foliage/Trees/Oak/Textures/T_AlderLeafs2_N}
  NamedRerouteDeclaration(None=@n8.B) {name=Glow Normal Read, nodeColor=<rgba 0.6172 1 0 1>}
  n9 = Multiply(A=NamedRerouteUsage(), B=NamedRerouteUsage())
  n10 = ScalarParameter() {defaultValue=1, group=Glow, parameterName=FresnelExp, sortPriority=2}
  n11 = Saturate(None=Add(A=LinearInterpolate(Alpha=@n9), B=Fresnel(ExponentIn=@n10) {baseReflectFraction=0, exponent=1}))
  n12 = CurveAtlasRowParameter(CurveTime=@n11) {
   atlas=/Game/Curie/Effects/temp/Curves/CA_Curie_Fire_Visuals
   curve=/Game/Curie/Effects/temp/Curves/Curve_Fiery
   defaultValue=2
   group=Glow
   parameterName=Curie
  }
  n13 = Multiply(A=@n12, B=Add(A=@n11) {constB=0.1})
  n14 = Saturate(None=@n13)
  n15 = ShadingPathSwitch(
   Default=MaterialFunctionCall(
    LDR Input=@n14
    HDR Tint=@n13
    HDR Intensity Multiplier=ScalarParameter() {defaultValue=1, group=Glow, parameterName=HDRIntensity, sortPriority=3}
   ) {materialFunction=&DeriveHDRfromLDR}.RemappedRange
   Forward=@n14
   Mobile=@n14
  )
  n16 = DynamicParameter() {defaultValue=<rgba 0 0 1 1>, paramNames [Glow,Burnt,Param3,Param4]}
  NamedRerouteDeclaration(None=MaterialFunctionCall(Emissive=Multiply(A=@n15, B=@n16.Glow)) {materialFunction=&MF_FX_EmissiveAdjustments}.Result) {name=Emissive, nodeColor=<rgba 1 0 0.1875 1>}
  NamedRerouteUsage()
  ScalarParameter() {defaultValue=0.9, desc=Estimated value from Ateria Leaves Base Material, parameterName=Opacity Value}
 ]
}
