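# Two-sided subsurface material for toon-style leaves with an outline, used on Niagara mesh particles in tree destruction effects.
# Vertex color red passes through a power, multiply and saturate to form a mask that blends particle color toward black, apparently painting the outline without textures.
# A shading-path and quality switch also outputs that mask, replacing it with a constant zero on mobile and low quality.
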Material /Game/Effects/Fort_Effects/World/Destruction/Destruction2/Trees/M_ToonLeafwithOUtline
graph {
 properties {
  bAutomaticallySetUsageInEditor=false
  bUsedWithNiagaraMeshParticles=true
  shadingModel=MSM_Subsurface
  twoSided=true
  usedShadingModels=Subsurface
 }
 nodes [
  n1 = VertexColor()
  n2 = Saturate(None=Multiply(A=Power(Base=@n1.R)) {constB=4})
  LinearInterpolate(A=ParticleColor().RGB, Alpha=@n2) {constB=0}
  n3 = Constant()
  ShadingPathSwitch(Default=QualitySwitch(Default=@n2, Low=@n3), Mobile=@n3)
 ]
}
