NiagaraSystem /Game/Effects/Fort_Effects/World/Destruction/Destruction2/Trees/NS_Destruction_AsteriaTree_BirchSnow
dependencies {
 usedDataInterfaces [
  &NiagaraDataInterfaceMeshRendererInfo
  &NiagaraDataInterfaceCurve
  &NiagaraDataInterfaceCollisionQuery
  /Script/ChaosNiagara.NiagaraDataInterfacePhysicsField
  &NiagaraDataInterfaceLandscape
 ]
 usedDynamicInputs []
 usedModules [&SystemState]
}
emitters {
 Leaves {
  simTarget=GPUComputeSim
  emitterUpdateScript [
   &EmitterState()
   &SpawnBurst_Instantaneous(
    Spawn Count=&MakeIntFromFloat(Float=$User.LeafAmount)
    Use Spawn Probability=false
    UseLoopCountLimit=false
   )
  ]
  particleSpawnScript [
   &InitializeParticle(
    Color=$User.LeafBaseColor
    Lifetime Max=3.5
    Lifetime Min=1.5
    Lifetime Mode=::NewEnumerator1
    Mesh Renderer Array Visibility Mode=::NewEnumerator3
    Mesh Renderer Info=DataInterface() {propertyValues {MeshRenderer=&self:Leaves_1.NiagaraMeshRendererProperties_0}}
    Mesh Scale Mode=::NewEnumerator1
    Mesh Uniform Scale Max=8
    Mesh Uniform Scale Min=3
    UsePositionOffset=false
    Write Material Random=true
   )
   &ShapeLocation(
    Cylinder Height=$User.TreeCylinderHeight
    Cylinder Height Midpoint=0
    Cylinder Radius=$User.TreeCylinderRadius
    MaskSpawnGroup=false
    Offset=$User.TreeCylinderOffset
    Offset Mode=::NewEnumerator0
    Override Seed=false
    Shape Primitive=::NewEnumerator1
    Surface Only=false
    UseOwnerScale=true
   )
   &ShapeLocation(MaskSpawnGroup=false, Override Seed=false, Sphere Radius=25, UseOwnerScale=true) {module=ShapeLocation001}
   &InitialMeshOrientation(Use Rotation Vector=false)
   &PointForce(
    Force Origin=&AddVectorToPosition(
     Position=/Niagara/DynamicInputs/Helpers/SimulationPosition()
     Vector=&Add_Vector(
      A=$User.TreeCylinderOffset
      B=&MakeVector(X=0, Y=0, Z=&Multiply_Float(A=$User.TreeCylinderRadius, B=0.5))
     )
    )
    Force Strength=&RandomRangeFloat(Minimum=333, Maximum=1000)
    Module.Use Falloff=false
    Module.UseLinearForce=false
   )
   &LinearForce(Force=&Multiply_VectorByFloat(
     Vector=$User.HitVector
     Float=&MakeCustomFloatFromBool(
      Bool=$User.bIsInstantDeath
      True Float=&RandomRangeFloat(Minimum=250, Maximum=1000)
      False Float=&RandomRangeFloat(Minimum=500, Maximum=1250)
     )
    ))
   &ApplyInitialForces()
  ]
  particleUpdateScript [
   &ParticleState()
   &ScaleMeshSize(Scale Factor=&VectorFromFloat(Value=&FloatFromCurve(FloatCurve=DataInterface() {propertyValues {Curve {keys [<at 0.8 1 RCIM_Cubic>,<at 1 0 RCIM_Cubic>]}}})))
   &Collision(
    Age Colliding Particles=false
    Collision Enabled=&SetBoolByFloatComparison(A=$Particles.Age, B=0.1)
    Control Roll On Collision=true
    Friction=0.333
    Friction During a Bounce=0.3
    GPU Collision Type=::NewEnumerator2
    Mesh Dimensions:Vector3f=<4 2 0>
    Module.Control Rest State Via Surface Normal=false
    Module.EnableMaxCollisionCount=false
    Module.Randomize Collision Normal Vector=true
    Radius Calculation Type=::NewEnumerator1
    Restitution=0
    ScaleRateOfRotation=false
   )
   &GravityForce()
   &WindForce(
    Ground Mask Elevation Mode:ENiagaraWindGroundMaskMode=::NewEnumerator1
    Use Ground Mask=true
    Use Physics Field=true
    Wind Speed=<1000 0 100>
    Wind Speed Scale=&MakeCustomFloatFromBool(Bool=$StackContext.HasCollided, True Float=0, False Float=0.1)
   )
   &AerodynamicDrag(
    Aerodynamic Drag=0.2
    Aerodynamic Rotational Drag=0.05
    Drag Shape Curve=DataInterface() {propertyValues {Curve {keys [
        {interpMode=RCIM_Cubic, tangentMode=RCTM_User, time=0, value=0.1}
        {arriveTangent=2.0799, interpMode=RCIM_Cubic, leaveTangent=2.0799, tangentMode=RCTM_User, time=1, value=1}
       ]}}}
    Lift Shape Curve=DataInterface() {propertyValues {Curve {keys [
        {interpMode=RCIM_Cubic, tangentMode=RCTM_User, time=0, value=0}
        {interpMode=RCIM_Cubic, tangentMode=RCTM_User, time=0.5, value=1}
        {tangentMode=RCTM_User, time=1, value=0}
       ]}}}
    Rotation Strength:NiagaraFloat=&RandomRangeFloat(Minimum=3, Maximum=8)
    Use Linear Drag=true
    Use Rotational Drag=true
   )
   &SolveForcesAndVelocity(Clamp Velocity=false, Limit Acceleration=false)
   &DynamicMaterialParameters(
    Index 0 Param 1=&MakeCustomFloatFromBool(
     Bool=$User.bIsCurie
     True Float=&FloatFromCurve(FloatCurve=DataInterface() {propertyValues {Curve {keys [
          {arriveTangent=-9.3075, interpMode=RCIM_Cubic, leaveTangent=-9.3075, tangentMode=RCTM_User, time=0, value=2}
          {arriveTangent=-2.0574, interpMode=RCIM_Cubic, leaveTangent=-2.0574, tangentMode=RCTM_User, time=0.2, value=0.625}
          <at 0.75 0 RCIM_Cubic>
         ]}}})
     False Float=0
    )
    Index 0 Param 2=&MakeCustomFloatFromBool(Bool=$User.bIsCurie, True Float=1, False Float=0)
    Module.Param0WriteEnabled=true
    Module.Param1WriteEnabled=true
    Module.Param2WriteEnabled=false
    Module.Param3WriteEnabled=false
   )
  ]
  rendererClasses [&NiagaraMeshRendererProperties]
  renderers [{rendererClass=&NiagaraMeshRendererProperties, rendererIndex=0}]
 }
 Leaves_Mobile {
  emitterUpdateScript [
   &EmitterState()
   &SpawnBurst_Instantaneous(
    Spawn Count=&MakeIntFromFloat(Float=$User.LeafAmount)
    Use Spawn Probability=false
    UseLoopCountLimit=false
   )
  ]
  particleSpawnScript [
   &InitializeParticle(
    Color=$User.LeafBaseColor
    Lifetime Max=3.5
    Lifetime Min=1.5
    Lifetime Mode=::NewEnumerator1
    Mesh Renderer Array Visibility Mode=::NewEnumerator3
    Mesh Renderer Info=DataInterface() {propertyValues {MeshRenderer=&self:Leaves_Mobile_2.NiagaraMeshRendererProperties_0}}
    Mesh Scale Mode=::NewEnumerator1
    Mesh Uniform Scale Max=8
    Mesh Uniform Scale Min=3
    UsePositionOffset=false
    Write Material Random=true
   )
   &ShapeLocation(
    Cylinder Height=$User.TreeCylinderHeight
    Cylinder Height Midpoint=0
    Cylinder Radius=$User.TreeCylinderRadius
    MaskSpawnGroup=false
    Offset=$User.TreeCylinderOffset
    Offset Mode=::NewEnumerator0
    Override Seed=false
    Shape Primitive=::NewEnumerator1
    Surface Only=false
    UseOwnerScale=true
   )
   &ShapeLocation(MaskSpawnGroup=false, Override Seed=false, Sphere Radius=25, UseOwnerScale=true) {module=ShapeLocation001}
   &InitialMeshOrientation(Use Rotation Vector=false)
   &PointForce(
    Force Origin=&AddVectorToPosition(
     Position=&SimulationPosition()
     Vector=&Add_Vector(
      A=$User.TreeCylinderOffset
      B=&MakeVector(X=0, Y=0, Z=&Multiply_Float(A=$User.TreeCylinderRadius, B=0.5))
     )
    )
    Force Strength=&RandomRangeFloat(Minimum=333, Maximum=1000)
    Module.Use Falloff=false
    Module.UseLinearForce=false
   )
   &LinearForce(Force=&Multiply_VectorByFloat(
     Vector=$User.HitVector
     Float=&MakeCustomFloatFromBool(
      Bool=$User.bIsInstantDeath
      True Float=&RandomRangeFloat(Minimum=250, Maximum=1000)
      False Float=&RandomRangeFloat(Minimum=500, Maximum=1250)
     )
    ))
   &ApplyInitialForces()
  ]
  particleUpdateScript [
   &ParticleState()
   &ScaleMeshSize(Scale Factor=&VectorFromFloat(Value=&FloatFromCurve(FloatCurve=DataInterface() {propertyValues {Curve {keys [<at 0.8 1 RCIM_Cubic>,<at 1 0 RCIM_Cubic>]}}})))
   &GravityForce()
   &WindForce(
    Ground Mask Elevation Mode:ENiagaraWindGroundMaskMode=::NewEnumerator1
    Use Ground Mask=true
    Use Physics Field=true
    Wind Speed=<1000 0 100>
    Wind Speed Scale=0.1
   )
   &AerodynamicDrag(
    Aerodynamic Drag=0.2
    Aerodynamic Rotational Drag=0.05
    Drag Shape Curve=DataInterface() {propertyValues {Curve {keys [
        {interpMode=RCIM_Cubic, tangentMode=RCTM_User, time=0, value=0.1}
        {arriveTangent=2.0799, interpMode=RCIM_Cubic, leaveTangent=2.0799, tangentMode=RCTM_User, time=1, value=1}
       ]}}}
    Lift Contribution=0.333
    Lift Shape Curve=DataInterface() {propertyValues {Curve {keys [
        {interpMode=RCIM_Cubic, tangentMode=RCTM_User, time=0, value=0}
        {interpMode=RCIM_Cubic, tangentMode=RCTM_User, time=0.5, value=1}
        {tangentMode=RCTM_User, time=1, value=0}
       ]}}}
    Rotation Strength:NiagaraFloat=&RandomRangeFloat(Minimum=3, Maximum=8)
    Use Linear Drag=true
    Use Rotational Drag=true
   )
   &SolveForcesAndVelocity(Clamp Velocity=false, Limit Acceleration=false)
   &DynamicMaterialParameters(
    Index 0 Param 1=&MakeCustomFloatFromBool(
     Bool=$User.bIsCurie
     True Float=&FloatFromCurve(FloatCurve=DataInterface() {propertyValues {Curve {keys [
          {arriveTangent=-9.3075, interpMode=RCIM_Cubic, leaveTangent=-9.3075, tangentMode=RCTM_User, time=0, value=2}
          {arriveTangent=-2.0574, interpMode=RCIM_Cubic, leaveTangent=-2.0574, tangentMode=RCTM_User, time=0.2, value=0.625}
          <at 0.75 0 RCIM_Cubic>
         ]}}})
     False Float=0
    )
    Index 0 Param 2=&MakeCustomFloatFromBool(Bool=$User.bIsCurie, True Float=1, False Float=0)
    Module.Param0WriteEnabled=true
    Module.Param1WriteEnabled=true
    Module.Param2WriteEnabled=false
    Module.Param3WriteEnabled=false
   )
  ]
  rendererClasses [&NiagaraMeshRendererProperties]
  renderers [{rendererClass=&NiagaraMeshRendererProperties, rendererIndex=0}]
 }
 SnowFlakes {
  simTarget=GPUComputeSim
  emitterUpdateScript [
   &EmitterState()
   &SpawnBurst_Instantaneous(
    Spawn Count=&Multiply_Int(
     A=&RandomRangeInt(Minimum=5, Maximum=6)
     B=&MakeIntFromFloat(Float=$User.LeafAmount)
    )
    Use Spawn Probability=false
    UseLoopCountLimit=false
   )
   &SpawnBurst_Instantaneous(
    Spawn Count=&Multiply_Int(
     A=&RandomRangeInt(Minimum=0, Maximum=1)
     B=&MakeIntFromFloat(Float=$User.LeafAmount)
    )
    Spawn Group=1
    Use Spawn Probability=false
    UseLoopCountLimit=false
   ) {module=SpawnBurst_Instantaneous001}
  ]
  particleSpawnScript [
   /SetVariables(Particles.VisibilityTag=&SpawnGroup()) {isSetParameters=true, module=SetVariables_#1}
   &InitializeParticle(
    Color Mode=::NewEnumerator0
    Lifetime Max=3.5
    Lifetime Min=2
    Lifetime Mode=::NewEnumerator1
    Mass=&RandomRangeFloat(Minimum=1, Maximum=3)
    Mass Mode=::NewEnumerator1
    Sprite Size Mode=::NewEnumerator1
    Uniform Sprite Size Max=&MaskFloatBySpawnGroup(Masked Float=20, Passthrough Float=8)
    Uniform Sprite Size Min=&MaskFloatBySpawnGroup(Masked Float=3, Passthrough Float=5)
    UsePositionOffset=false
    Write Material Random=true
   )
   &ShapeLocation(
    Cylinder Height=$User.TreeCylinderHeight
    Cylinder Height Midpoint=0
    Cylinder Radius=$User.TreeCylinderRadius
    MaskSpawnGroup=false
    Offset=$User.TreeCylinderOffset
    Offset Coordinate Space=::Local
    Offset Mode=::NewEnumerator0
    Override Seed=false
    Shape Primitive=::NewEnumerator1
    Surface Only=false
    UseOwnerScale=true
   )
   &PointForce(
    Force Origin=&AddVectorToPosition(
     Position=&SimulationPosition()
     Vector=&Multiply_Vector(A=$User.TreeCylinderOffset, B=&Multiply_VectorByFloat(Vector=<1 1 1>, Float=0.5))
    )
    Force Strength=&RandomRangeFloat(Minimum=333, Maximum=1000)
    Module.Use Falloff=false
    Module.UseLinearForce=false
   )
   &ApplyInitialForces()
  ]
  particleUpdateScript [
   &ParticleState()
   &GravityForce(Gravity=&MaskVectorBySpawnGroup(Masked Vector=<0 0 -250>, Passthrough Vector=<0 0 -980>))
   &Drag(Drag=&RandomRangeFloat(Minimum=1, Maximum=3), Use Linear Drag=true, Use Rotational Drag=true)
   &VectorNoiseForce(
    Force Amount=&MaskFloatBySpawnGroup(Masked Float=5000, Passthrough Float=0)
    Module.Use Random Noise Time Interval=false
   )
   &SolveForcesAndVelocity(Clamp Velocity=false, Limit Acceleration=false)
   &ScaleColor(
    Scale Alpha=&FloatFromCurve(FloatCurve=DataInterface() {propertyValues {Curve {keys [{arriveTangent=-1.6891, interpMode=RCIM_Cubic, leaveTangent=-1.6891, tangentMode=RCTM_User, time=0, value=1}, <at 1 0 RCIM_Cubic>]}}})
    ScaleA=true
    ScaleRGB=true
   )
  ]
  rendererClasses [&NiagaraSpriteRendererProperties]
  renderers [
   {rendererClass=&NiagaraSpriteRendererProperties}
   {rendererClass=&NiagaraSpriteRendererProperties, rendererIndex=1}
  ]
 }
 SnowFlakes_Mobile {
  emitterUpdateScript [
   &EmitterState()
   &SpawnBurst_Instantaneous(
    Spawn Count=&Multiply_Int(
     A=&RandomRangeInt(Minimum=5, Maximum=6)
     B=&MakeIntFromFloat(Float=$User.LeafAmount)
    )
    Use Spawn Probability=false
    UseLoopCountLimit=false
   )
   &SpawnBurst_Instantaneous(
    Spawn Count=&Multiply_Int(
     A=&RandomRangeInt(Minimum=0, Maximum=1)
     B=&MakeIntFromFloat(Float=$User.LeafAmount)
    )
    Spawn Group=1
    Use Spawn Probability=false
    UseLoopCountLimit=false
   ) {module=SpawnBurst_Instantaneous001}
  ]
  particleSpawnScript [
   /SetVariables(Particles.VisibilityTag=&SpawnGroup()) {isSetParameters=true, module=SetVariables_#1}
   &InitializeParticle(
    Color Mode=::NewEnumerator0
    Lifetime Max=3.5
    Lifetime Min=2
    Lifetime Mode=::NewEnumerator1
    Mass=&RandomRangeFloat(Minimum=1, Maximum=3)
    Mass Mode=::NewEnumerator1
    Sprite Size Mode=::NewEnumerator1
    Uniform Sprite Size Max=&MaskFloatBySpawnGroup(Masked Float=20, Passthrough Float=8)
    Uniform Sprite Size Min=&MaskFloatBySpawnGroup(Masked Float=3, Passthrough Float=5)
    UsePositionOffset=false
    Write Material Random=true
   )
   &ShapeLocation(
    Cylinder Height=$User.TreeCylinderHeight
    Cylinder Height Midpoint=0
    Cylinder Radius=$User.TreeCylinderRadius
    MaskSpawnGroup=false
    Offset=$User.TreeCylinderOffset
    Offset Coordinate Space=::Local
    Offset Mode=::NewEnumerator0
    Override Seed=false
    Shape Primitive=::NewEnumerator1
    Surface Only=false
    UseOwnerScale=true
   )
   &PointForce(
    Force Origin=&AddVectorToPosition(
     Position=&SimulationPosition()
     Vector=&Multiply_Vector(A=$User.TreeCylinderOffset, B=&Multiply_VectorByFloat(Vector=<1 1 1>, Float=0.5))
    )
    Force Strength=&RandomRangeFloat(Minimum=333, Maximum=1000)
    Module.Use Falloff=false
    Module.UseLinearForce=false
   )
   &ApplyInitialForces()
  ]
  particleUpdateScript [
   &ParticleState()
   &GravityForce(Gravity=&MaskVectorBySpawnGroup(Masked Vector=<0 0 -250>, Passthrough Vector=<0 0 -980>))
   &Drag(Drag=&RandomRangeFloat(Minimum=1, Maximum=3), Use Linear Drag=true, Use Rotational Drag=true)
   &VectorNoiseForce(
    Force Amount=&MaskFloatBySpawnGroup(Masked Float=5000, Passthrough Float=0)
    Module.Use Random Noise Time Interval=false
   )
   &SolveForcesAndVelocity(Clamp Velocity=false, Limit Acceleration=false)
   &ScaleColor(
    Scale Alpha=&FloatFromCurve(FloatCurve=DataInterface() {propertyValues {Curve {keys [{arriveTangent=-1.6891, interpMode=RCIM_Cubic, leaveTangent=-1.6891, tangentMode=RCTM_User, time=0, value=1}, <at 1 0 RCIM_Cubic>]}}})
    ScaleA=true
    ScaleRGB=true
   )
  ]
  rendererClasses [&NiagaraSpriteRendererProperties]
  renderers [
   {rendererClass=&NiagaraSpriteRendererProperties}
   {rendererClass=&NiagaraSpriteRendererProperties, rendererIndex=1}
  ]
 }
}
summary {
 systemName=NS_Destruction_AsteriaTree_BirchSnow
 emitters [
  {emitterName=Leaves, simTarget=GPUComputeSim, rendererClasses [&NiagaraMeshRendererProperties]}
  {emitterName=Leaves_Mobile, rendererClasses [&NiagaraMeshRendererProperties]}
  {emitterName=SnowFlakes, simTarget=GPUComputeSim, rendererClasses [&NiagaraSpriteRendererProperties]}
  {emitterName=SnowFlakes_Mobile, rendererClasses [&NiagaraSpriteRendererProperties]}
 ]
 userVariables [
  User.bIsCurie {type=NiagaraBool}
  User.bIsInstantDeath {type=NiagaraBool}
  User.HitVector {type=Vector3f, description="", defaultValue {struct=Vector3f, value=<0 0 0>}}
  User.LeafAmount {type=NiagaraFloat, defaultValue=300, description=""}
  User.LeafBaseColor {type=LinearColor, description="", defaultValue {struct=LinearColor, value=<rgba 0.1413 0.2052 0.0366 1>}}
  User.TreeCylinderHeight {type=NiagaraFloat, defaultValue=375, description=""}
  User.TreeCylinderOffset {type=Vector3f, description="", defaultValue {struct=Vector3f, value=<0 0 800>}}
  User.TreeCylinderRadius {type=NiagaraFloat, defaultValue=150, description=""}
 ]
}
