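# Niagara destruction effect for the Asteria palm tree (variant 2) that bursts falling palm-leaf meshes from the crown.
# Leaves spawn in a user-sized cylinder plus sphere, shift along their mesh axis, and are launched by point forces and HitVector, scaled by bIsInstantDeath.
# Gravity, wind and aerodynamic drag drive them until they shrink away; bIsCurie drives material parameters.
# The GPU emitter adds collision that cuts wind after impact, while the Mobile emitter omits collision.
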NiagaraSystem /Game/Effects/Fort_Effects/World/Destruction/Destruction2/Trees/NS_Destruction_AsteriaTree_Palm2
dependencies {
 usedDataInterfaces [
  &NiagaraDataInterfaceMeshRendererInfo
  &NiagaraDataInterfaceCurve
  &NiagaraDataInterfaceCollisionQuery
  /Script/ChaosNiagara.NiagaraDataInterfacePhysicsField
  &NiagaraDataInterfaceLandscape
 ]
 usedDynamicInputs []
 usedModules [&SystemState]
}
emitters {
 Leaves {
  simTarget=GPUComputeSim
  emitterUpdateScript [
   &EmitterState()
   &SpawnBurst_Instantaneous(
    Spawn Count=&MakeIntFromFloat(Float=$User.LeafAmount)
    Use Spawn Probability=false
    UseLoopCountLimit=false
   )
  ]
  particleSpawnScript [
   &InitializeParticle(
    Color Mode=::NewEnumerator0
    Lifetime Max=3.5
    Lifetime Min=2
    Lifetime Mode=::NewEnumerator1
    Mesh Renderer Array Visibility Mode=::NewEnumerator3
    Mesh Renderer Info=DataInterface() {propertyValues {MeshRenderer=&self:Leaves_1.NiagaraMeshRendererProperties_0}}
    Mesh Scale Mode=::NewEnumerator1
    Mesh Uniform Scale Max=10
    Mesh Uniform Scale Min=8
    UsePositionOffset=false
    Write Material Random=true
   )
   &ShapeLocation(
    Cylinder Height=$User.TreeCylinderHeight
    Cylinder Height Midpoint=0
    Cylinder Radius=$User.TreeCylinderRadius
    MaskSpawnGroup=false
    Offset=$User.TreeCylinderOffset
    Offset Mode=::NewEnumerator0
    Override Seed=false
    Shape Primitive=::NewEnumerator1
    Surface Only=false
    UseOwnerScale=true
   )
   &ShapeLocation(MaskSpawnGroup=false, Override Seed=false, Sphere Radius=$User.TreeSpheriodRadius, UseOwnerScale=true) {module=ShapeLocation001}
   &InitialMeshOrientation(
    Mesh Orientation Mode=::NewEnumerator1
    Rotation=&RandomRangeVector(Minimum=<0 0 0>, Maximum=<0 0 1>)
    Use Rotation Vector=true
   )
   &PointForce(
    Force Origin=&AddVectorToPosition(Position=/Niagara/DynamicInputs/Helpers/SimulationPosition(), Vector=<0 0 1000>)
    Force Strength=&RandomRangeFloat(Minimum=333, Maximum=1000)
    Module.Use Falloff=false
    Module.UseLinearForce=false
   ) {enabled=false}
   /SetVariables(Particles.Position=&AddVectorToPosition(
     Position=$Particles.Position
     Vector=&Multiply_VectorByFloat(
      Vector=&ReturnMeshOrientationAxis(Mesh Orientation Axis=::XAxis, Mesh Orientation Quat=$Particles.MeshOrientation)
      Float=-170
     )
    )) {isSetParameters=true, module=SetVariables_#1}
   &PointForce(
    Force Origin=&AddVectorToPosition(
     Position=$Particles.Position
     Vector=&Multiply_VectorByFloat(
      Vector=&ReturnMeshOrientationAxis(Mesh Orientation Axis=::XAxis, Mesh Orientation Quat=$Particles.MeshOrientation)
      Float=170
     )
    )
    Force Strength=&RandomRangeFloat(Minimum=100, Maximum=500)
    Module.Use Falloff=false
    Module.UseLinearForce=false
   ) {module=PointForce001}
   &LinearForce(Force=&Multiply_VectorByFloat(
     Vector=$User.HitVector
     Float=&MakeCustomFloatFromBool(
      Bool=$User.bIsInstantDeath
      True Float=&RandomRangeFloat(Minimum=250, Maximum=1000)
      False Float=&RandomRangeFloat(Minimum=500, Maximum=1250)
     )
    ))
   &ApplyInitialForces()
   &InitialMeshOrientation(
    Mesh Orientation Mode=::NewEnumerator7
    Rotation=&RandomRangeVector(Minimum=<0.1 0.1 0.1>, Maximum=<-0.1 -0.1 -0.1>)
    Use Rotation Vector=true
   ) {module=InitialMeshOrientation001}
  ]
  particleUpdateScript [
   &ParticleState()
   &ScaleMeshSize(Scale Factor=&VectorFromFloat(Value=&FloatFromCurve(FloatCurve=DataInterface() {propertyValues {Curve {keys [<at 0.8 1 RCIM_Cubic>,<at 1 0 RCIM_Cubic>]}}})))
   &Collision(
    Age Colliding Particles=false
    Collision Enabled=&SetBoolByFloatComparison(A=$Particles.Age, B=0.1)
    Control Roll On Collision=true
    Friction=0.333
    Friction During a Bounce=0.3
    GPU Collision Type=::NewEnumerator2
    Mesh Dimensions:Vector3f=<37 14 4>
    Method for Calculating Particle Radius:ENiagaraCalculateRadiusOptions=::NewEnumerator1
    Module.Control Rest State Via Surface Normal=false
    Module.EnableMaxCollisionCount=false
    Module.Randomize Collision Normal Vector=true
    Radius Calculation Type=::NewEnumerator1
    Restitution=0
    ScaleRateOfRotation=false
   )
   &GravityForce()
   &WindForce(
    Ground Mask Elevation Mode:ENiagaraWindGroundMaskMode=::NewEnumerator1
    Use Ground Mask=true
    Use Physics Field=true
    Wind Speed=<1000 0 100>
    Wind Speed Scale=&MakeCustomFloatFromBool(Bool=$StackContext.HasCollided, True Float=0, False Float=0.1)
   )
   &AerodynamicDrag(
    Aerodynamic Drag=0.05
    Aerodynamic Rotational Drag=0.005
    Drag Shape Curve=DataInterface() {propertyValues {Curve {keys [
        {interpMode=RCIM_Cubic, tangentMode=RCTM_User, time=0, value=0.1}
        {arriveTangent=2.0799, interpMode=RCIM_Cubic, leaveTangent=2.0799, tangentMode=RCTM_User, time=1, value=1}
       ]}}}
    Lift Shape Curve=DataInterface() {propertyValues {Curve {keys [
        {interpMode=RCIM_Cubic, tangentMode=RCTM_User, time=0, value=0}
        {interpMode=RCIM_Cubic, tangentMode=RCTM_User, time=0.5, value=1}
        {tangentMode=RCTM_User, time=1, value=0}
       ]}}}
    Rotation Strength:NiagaraFloat=&RandomRangeFloat(Minimum=0.25, Maximum=1)
    Use Linear Drag=true
    Use Rotational Drag=true
   )
   &SolveForcesAndVelocity(Clamp Velocity=false, Limit Acceleration=false)
   &DynamicMaterialParameters(
    Index 0 Param 1=&MakeCustomFloatFromBool(
     Bool=$User.bIsCurie
     True Float=&FloatFromCurve(FloatCurve=DataInterface() {propertyValues {Curve {keys [
          {arriveTangent=-9.3075, interpMode=RCIM_Cubic, leaveTangent=-9.3075, tangentMode=RCTM_User, time=0, value=2}
          {arriveTangent=-2.0574, interpMode=RCIM_Cubic, leaveTangent=-2.0574, tangentMode=RCTM_User, time=0.2, value=0.625}
          <at 0.75 0 RCIM_Cubic>
         ]}}})
     False Float=0
    )
    Index 0 Param 2=&MakeCustomFloatFromBool(Bool=$User.bIsCurie, True Float=1, False Float=0)
    Module.Param0WriteEnabled=true
    Module.Param1WriteEnabled=true
    Module.Param2WriteEnabled=false
    Module.Param3WriteEnabled=false
   )
  ]
  rendererClasses [&NiagaraMeshRendererProperties]
  renderers [{rendererClass=&NiagaraMeshRendererProperties, rendererIndex=0}]
 }
 Leaves_Mobile {
  emitterUpdateScript [
   &EmitterState()
   &SpawnBurst_Instantaneous(
    Spawn Count=&MakeIntFromFloat(Float=$User.LeafAmount)
    Use Spawn Probability=false
    UseLoopCountLimit=false
   )
  ]
  particleSpawnScript [
   &InitializeParticle(
    Color Mode=::NewEnumerator0
    Lifetime Max=3.5
    Lifetime Min=2
    Lifetime Mode=::NewEnumerator1
    Mesh Renderer Array Visibility Mode=::NewEnumerator3
    Mesh Renderer Info=DataInterface() {propertyValues {MeshRenderer=&self:Leaves_Mobile_2.NiagaraMeshRendererProperties_0}}
    Mesh Scale Mode=::NewEnumerator1
    Mesh Uniform Scale Max=10
    Mesh Uniform Scale Min=8
    UsePositionOffset=false
    Write Material Random=true
   )
   &ShapeLocation(
    Cylinder Height=$User.TreeCylinderHeight
    Cylinder Height Midpoint=0
    Cylinder Radius=$User.TreeCylinderRadius
    MaskSpawnGroup=false
    Offset=$User.TreeCylinderOffset
    Offset Mode=::NewEnumerator0
    Override Seed=false
    Shape Primitive=::NewEnumerator1
    Surface Only=false
    UseOwnerScale=true
   )
   &ShapeLocation(MaskSpawnGroup=false, Override Seed=false, Sphere Radius=$User.TreeSpheriodRadius, UseOwnerScale=true) {module=ShapeLocation001}
   &InitialMeshOrientation(
    Mesh Orientation Mode=::NewEnumerator1
    Rotation=&RandomRangeVector(Minimum=<0 0 0>, Maximum=<0 0 1>)
    Use Rotation Vector=true
   )
   &PointForce(
    Force Origin=&AddVectorToPosition(Position=&SimulationPosition(), Vector=<0 0 1000>)
    Force Strength=&RandomRangeFloat(Minimum=333, Maximum=1000)
    Module.Use Falloff=false
    Module.UseLinearForce=false
   )
   &LinearForce(Force=&Multiply_VectorByFloat(
     Vector=$User.HitVector
     Float=&MakeCustomFloatFromBool(
      Bool=$User.bIsInstantDeath
      True Float=&RandomRangeFloat(Minimum=250, Maximum=1000)
      False Float=&RandomRangeFloat(Minimum=500, Maximum=1250)
     )
    ))
   &ApplyInitialForces()
   /SetVariables(Particles.Position=&AddVectorToPosition(
     Position=$Particles.Position
     Vector=&Multiply_VectorByFloat(
      Vector=&ReturnMeshOrientationAxis(Mesh Orientation Axis=::XAxis, Mesh Orientation Quat=$Particles.MeshOrientation)
      Float=-170
     )
    )) {isSetParameters=true, module=SetVariables_#1}
  ]
  particleUpdateScript [
   &ParticleState()
   &ScaleMeshSize(Scale Factor=&VectorFromFloat(Value=&FloatFromCurve(FloatCurve=DataInterface() {propertyValues {Curve {keys [<at 0.8 1 RCIM_Cubic>,<at 1 0 RCIM_Cubic>]}}})))
   &GravityForce()
   &WindForce(
    Ground Mask Elevation Mode:ENiagaraWindGroundMaskMode=::NewEnumerator1
    Use Ground Mask=true
    Use Physics Field=true
    Wind Speed=<1000 0 100>
    Wind Speed Scale=0.1
   )
   &AerodynamicDrag(
    Aerodynamic Drag=0.05
    Aerodynamic Rotational Drag=0.05
    Drag Shape Curve=DataInterface() {propertyValues {Curve {keys [
        {interpMode=RCIM_Cubic, tangentMode=RCTM_User, time=0, value=0.1}
        {arriveTangent=2.0799, interpMode=RCIM_Cubic, leaveTangent=2.0799, tangentMode=RCTM_User, time=1, value=1}
       ]}}}
    Lift Contribution=0.333
    Lift Shape Curve=DataInterface() {propertyValues {Curve {keys [
        {interpMode=RCIM_Cubic, tangentMode=RCTM_User, time=0, value=0}
        {interpMode=RCIM_Cubic, tangentMode=RCTM_User, time=0.5, value=1}
        {tangentMode=RCTM_User, time=1, value=0}
       ]}}}
    Rotation Strength:NiagaraFloat=&RandomRangeFloat(Minimum=0.25, Maximum=1)
    Use Linear Drag=true
    Use Rotational Drag=true
   )
   &SolveForcesAndVelocity(Clamp Velocity=false, Limit Acceleration=false)
   &DynamicMaterialParameters(
    Index 0 Param 1=&MakeCustomFloatFromBool(
     Bool=$User.bIsCurie
     True Float=&FloatFromCurve(FloatCurve=DataInterface() {propertyValues {Curve {keys [
          {arriveTangent=-9.3075, interpMode=RCIM_Cubic, leaveTangent=-9.3075, tangentMode=RCTM_User, time=0, value=2}
          {arriveTangent=-2.0574, interpMode=RCIM_Cubic, leaveTangent=-2.0574, tangentMode=RCTM_User, time=0.2, value=0.625}
          <at 0.75 0 RCIM_Cubic>
         ]}}})
     False Float=0
    )
    Index 0 Param 2=&MakeCustomFloatFromBool(Bool=$User.bIsCurie, True Float=1, False Float=0)
    Module.Param0WriteEnabled=true
    Module.Param1WriteEnabled=true
    Module.Param2WriteEnabled=false
    Module.Param3WriteEnabled=false
   )
  ]
  rendererClasses [&NiagaraMeshRendererProperties]
  renderers [{rendererClass=&NiagaraMeshRendererProperties, rendererIndex=0}]
 }
}
summary {
 systemName=NS_Destruction_AsteriaTree_Palm2
 emitters [
  {emitterName=Leaves, simTarget=GPUComputeSim, rendererClasses [&NiagaraMeshRendererProperties]}
  {emitterName=Leaves_Mobile, rendererClasses [&NiagaraMeshRendererProperties]}
 ]
 userVariables [
  User.bIsCurie {type=NiagaraBool}
  User.bIsInstantDeath {type=NiagaraBool}
  User.HitVector {type=Vector3f, description="", defaultValue {struct=Vector3f, value=<0 0 0>}}
  User.LeafAmount {type=NiagaraFloat, defaultValue=10, description=""}
  User.LeafBaseColor1 {type=LinearColor, description="", defaultValue {struct=LinearColor, value=<rgba 0.1413 0.2052 0.0366 1>}}
  User.LeafHueVariance {type=NiagaraFloat, defaultValue=-0.02, description=""}
  User.LeafValueVariance {type=NiagaraFloat, defaultValue=0.3, description=""}
  User.NoVariance {type=NiagaraFloat, defaultValue=1, description=""}
  User.TreeCylinderHeight {type=NiagaraFloat, defaultValue=100, description=""}
  User.TreeCylinderOffset {type=Vector3f, description="", defaultValue {struct=Vector3f, value=<150 0 1100>}}
  User.TreeCylinderRadius {type=NiagaraFloat, defaultValue=200, description=""}
  User.TreeSpheriodRadius {type=NiagaraFloat, defaultValue=0, description=""}
 ]
}
